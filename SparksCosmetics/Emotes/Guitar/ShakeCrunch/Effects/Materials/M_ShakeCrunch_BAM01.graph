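# Unlit masked Niagara sprite material for the ShakeCrunch guitar emote's BAM effect, animating two 2x2 flipbook textures over time.
# The base flipbook, switchable to static UVs by parameter, is multiplied by a second flipbook running at a scaled, offset rate, with adjustable presence.
# The result passes through saturation, particle color and brightness into MF_FX_EmissiveAdjustments.
# Particle alpha multiplies the base texture alpha to shape the sprite.
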Material /SparksCosmetics/Emotes/Guitar/ShakeCrunch/Effects/Materials/M_ShakeCrunch_BAM01
graph {
 properties {bAutomaticallySetUsageInEditor=false, bUsedWithNiagaraSprites=true, blendMode=BLEND_Masked, shadingModel=MSM_Unlit, usedShadingModels=Unlit}
 nodes [
  n1 = Time()
  n2 = ScalarParameter() {defaultValue=2.8, parameterName=Speed}
  n3 = Frac(None=Multiply(A=@n2, B=@n1))
  n4 = Constant() {r=2}
  n5 = TextureCoordinate()
  n6 = TextureSampleParameter2D(UVs=LinearInterpolate(
    A=TextureCoordinate()
    B=MaterialFunctionCall("Animation  Phase (0-1)"=@n3, Number of Rows=@n4, Number of Columns=@n4, UVs=@n5) {materialFunction=&FlipBook}.UVs
    Alpha=ScalarParameter() {defaultValue=1, parameterName=UseBaseFlipbook?}
   )) {parameterName=BaseTexture, texture=/SparksCosmetics/Emotes/Guitar/ShakeCrunch/Effects/Textures/T_ShakeCrunch_Flipbook01}
  n7 = Multiply(
   A=Add(A=@n1, B=ScalarParameter() {defaultValue=0.5, parameterName=TextureFrameOffset})
   B=Multiply(A=@n2, B=ScalarParameter() {defaultValue=0.3, parameterName=TextureFrameMultiplier}) {constB=2}
  )
  n8 = Constant() {r=2}
  n9 = TextureSample(UVs=MaterialFunctionCall("Animation  Phase (0-1)"=Frac(None=@n7), Number of Rows=@n8, Number of Columns=@n8, UVs=TextureCoordinate()) {materialFunction=&FlipBook}.UVs) {texture=/SparksCosmetics/Emotes/Guitar/ShakeCrunch/Effects/Textures/T_ShakeCrunch_Flipbook02}
  n10 = ScalarParameter() {defaultValue=0.9, parameterName=TexturePresence}
  n11 = Desaturation(
   None=Multiply(A=@n6.RGB, B=LinearInterpolate(B=@n9.RGB, Alpha=@n10) {constA=1})
   Fraction=OneMinus(None=ScalarParameter() {defaultValue=1, parameterName=BaseTexture_Saturation})
  )
  n12 = ParticleColor()
  MaterialFunctionCall(Emissive=Multiply(A=ScalarParameter() {defaultValue=1.5, parameterName=Brightness}, B=Multiply(A=@n11, B=@n12.RGB))) {materialFunction=&MF_FX_EmissiveAdjustments}
  Multiply(A=@n12.A, B=@n6.A)
 ]
}
