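# M_ShakeCrunch_BlankStones is a plain Default Lit material for stone mesh particles in the Sparks ShakeCrunch guitar emote effects.
# Its only expression multiplies particle color RGB by a dark grey constant, so the tint comes from the particle system with no textures.
# It is flagged for Niagara mesh particles, with automatic usage setting in the editor turned off.
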Material /SparksCosmetics/Emotes/Guitar/ShakeCrunch/Effects/Materials/M_ShakeCrunch_BlankStones
graph {
 properties {bAutomaticallySetUsageInEditor=false, bUsedWithNiagaraMeshParticles=true, usedShadingModels=Default Lit}
 nodes [Multiply(A=ParticleColor().RGB, B=Constant3Vector() {constant=<rgba 0.2413 0.2413 0.2413 0>})]
}
