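# Masked, unlit, two-sided Niagara sprite and mesh particle material for effects in the ShakeCrunch guitar emote.
# Two flipbooks, BAM shards for alpha and silhouette grime for emissive, loop at speeds scaled by a shared master control, with per-particle random time offsets.
# Presence parameters blend each flipbook against white, and the alpha texture's channel is selectable.
# The emissive result is multiplied by particle color and a tint, then fed through MF_FX_EmissiveAdjustments.
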Material /SparksCosmetics/Emotes/Guitar/ShakeCrunch/Effects/Materials/M_ShakeCrunch_Flipbook
graph {
 properties {
  bAutomaticallySetUsageInEditor=false
  bUsedWithNiagaraMeshParticles=true
  bUsedWithNiagaraSprites=true
  blendMode=BLEND_Masked
  shadingModel=MSM_Unlit
  twoSided=true
  usedShadingModels=Unlit
 }
 nodes [
  DynamicParameter()
  n1 = ScalarParameter() {defaultValue=0.7, group=0. Master Controls, parameterName=Flipbook_MasterSpeed}
  n2 = Time()
  n3 = ParticleRandom()
  n4 = Multiply(
   A=Multiply(A=@n1, B=ScalarParameter() {defaultValue=1, group=2. Alpha, parameterName=Alpha_FlipbookSpeed})
   B=Add(A=@n2, B=Multiply(A=@n3) {constB=11.77})
  )
  n5 = TextureSampleParameter2D(UVs=MaterialFunctionCall(
    "Animation  Phase (0-1)"=Frac(None=@n4)
    Number of Rows=ScalarParameter() {defaultValue=1, group=2. Alpha, parameterName=Alpha_Flipbook_X_Rows}
    Number of Columns=ScalarParameter() {defaultValue=2, group=2. Alpha, parameterName=Alpha_Flipbook_Y_Columns}
    UVs=TextureCoordinate()
   ) {materialFunction=&FlipBook}.UVs) {
   group=2. Alpha
   parameterName=Tex_AlphaFlipbook
   texture=/SparksCosmetics/Emotes/Guitar/ShakeCrunch/Effects/Textures/T_ShakeCrunch_BAMShards_Flipbook
  }
  n6 = ChannelMaskParameter(None=@n5.RGBA) {group=2. Alpha, maskChannel=Alpha, parameterName=Alpha_ChooseFlipbookChannel, defaultValue=<rgba 0 0 0 1>}
  LinearInterpolate(B=@n6.R, Alpha=ScalarParameter() {defaultValue=1, group=2. Alpha, parameterName=Alpha_FlipbookPresence}) {constA=1}
  n7 = ScalarParameter() {defaultValue=1, group=1. Emissive, parameterName=Emissive_FlipbookSpeed}
  n8 = Add(A=Multiply(A=ParticleRandom()) {constB=10}, B=Multiply(A=Multiply(A=@n7, B=@n1), B=Add(B=@n2) {constA=0.5}))
  n9 = TextureSampleParameter2D(UVs=MaterialFunctionCall(
    "Animation  Phase (0-1)"=Frac(None=@n8)
    Number of Rows=ScalarParameter() {defaultValue=4, group=1. Emissive, parameterName=Emissive_Flipbook_X_Rows}
    Number of Columns=ScalarParameter() {defaultValue=4, group=1. Emissive, parameterName=Emissive_Flipbook_Y_Rows}
    UVs=TextureCoordinate()
   ) {materialFunction=&FlipBook}.UVs) {
   group=1. Emissive
   parameterName=Tex_EmissiveFlipbook
   texture=/SparksCosmetics/Emotes/Guitar/ShakeCrunch/Effects/Textures/T_ShakeCrunch_SilhouetteGrime
  }
  n10 = LinearInterpolate(B=@n9.R, Alpha=ScalarParameter() {defaultValue=1, group=1. Emissive, parameterName=Emissive_FlipbookPresence}) {constA=1}
  n11 = ParticleColor()
  MaterialFunctionCall(Emissive=Multiply(A=VectorParameter() {group=1. Emissive, parameterName=EmissiveColor, defaultValue=<rgba 1 1 1 1>}.RGB, B=Multiply(A=@n10, B=@n11.RGB))) {materialFunction=&MF_FX_EmissiveAdjustments}
  ScalarParameter() {defaultValue=0.1, parameterName=AnimationPhase}
  Transform(None=MaterialFunctionCall(
    R=ScalarParameter() {defaultValue=5, parameterName=R}
    G=ScalarParameter() {defaultValue=6, parameterName=G}
    B=ScalarParameter() {defaultValue=9, parameterName=B}
   ) {materialFunction=&AppendMany}.RGB) {transformSourceType=TRANSFORMSOURCE_World, transformType=TRANSFORM_Local}
 ]
}
