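# Translucent, unlit, two-sided particle erosion material for Niagara sprites and meshes in the Sparks ShakeCrunch guitar emote effects.
# Particle color RGB runs through the shared FX emissive adjustments function.
# The mask texture, rings by default, has its green channel reduced by a per-particle Erode dynamic parameter, then divided by a softness value and saturated.
# That eroded mask multiplies texture alpha and particle alpha to form opacity.
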Material /SparksCosmetics/Emotes/Guitar/ShakeCrunch/Effects/Materials/M_Sparks_ParticleErode
graph {
 properties {
  bAutomaticallySetUsageInEditor=false
  bUsedWithNiagaraMeshParticles=true
  bUsedWithNiagaraSprites=true
  blendMode=BLEND_Translucent
  shadingModel=MSM_Unlit
  twoSided=true
  usedShadingModels=Unlit
 }
 nodes [
  MaterialFunctionCall(Emissive=ParticleColor().RGB) {materialFunction=&MF_FX_EmissiveAdjustments}
  n1 = TextureSampleParameter2D() {
   parameterName=Opacity Mask
   samplerSource=SSM_Clamp_WorldGroupSettings
   texture=/SparksCosmetics/Emotes/Guitar/ShakeCrunch/Effects/Textures/T_ShakeCrunch_Rings
  }
  n2 = Divide(
   A=Subtract(A=@n1.G, B=DynamicParameter() {defaultValue=<rgba 1 0 1 0>, paramNames [Erode, "", "", ""]}.Erode)
   B=ScalarParameter() {defaultValue=0.1, group=Opacity, parameterName=OpacitySoftness}
  )
  Multiply(A=Multiply(A=@n1.A, B=Saturate(None=@n2)), B=ParticleColor().A)
 ]
}
