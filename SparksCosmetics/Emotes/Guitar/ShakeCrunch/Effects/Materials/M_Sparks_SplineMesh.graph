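# Masked, two-sided unlit material for spline-mesh lightning sparks in the ShakeCrunch guitar emote, usable on Niagara mesh particles.
# SplineThicken widens strands using vertex-color channel weights.
# A cosine flicker with a stepped blink, seeded by particle random or object position, scales an added world position offset.
# A panning lightning alpha texture, trimmed at both U edges, forms the mask, while a color parameter feeds emissive adjustments.
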Material /SparksCosmetics/Emotes/Guitar/ShakeCrunch/Effects/Materials/M_Sparks_SplineMesh
graph {
 properties {
  bAutomaticallySetUsageInEditor=false
  bCastRayTracedShadows=false
  bUsedWithNiagaraMeshParticles=true
  blendMode=BLEND_Masked
  shadingModel=MSM_Unlit
  twoSided=true
  usedShadingModels=Unlit
 }
 nodes [
  n1 = StaticBool()
  n2 = VectorParameter() {parameterName=VertColMask_ThickenMultiplier, defaultValue=<rgba 10 5 1.5 1>}
  n3 = VertexColor()
  n4 = Multiply(A=@n2.R, B=@n3.R)
  n5 = Multiply(A=@n2.G, B=@n3.G)
  n6 = Multiply(
   A=ScalarParameter() {defaultValue=1, group=Spline Settings, parameterName=Spline Width Base, sortPriority=1}
   B=Add(A=Add(A=@n4, B=@n5), B=Multiply(A=@n2.B, B=@n3.B))
  )
  n7 = MaterialFunctionCall(Color=ObjectPositionWS()) {materialFunction=&SplitComponents}
  n8 = Add(A=Add(A=@n7.R, B=@n7.R), B=Multiply(A=@n7.B) {constB=777.777})
  n9 = Add(
   A=StaticSwitch(True=ParticleRandom(), False=Divide(A=@n8) {constB=1}, Value=StaticBoolParameter() {defaultValue=true, parameterName=IsParticle})
   B=Multiply(A=Time(), B=ScalarParameter() {defaultValue=3, desc=Waveform Speed, parameterName=FlickerSpeed}) {constB=0.1}
  )
  n10 = Constant4Vector() {constant=<rgba 0.35 1.6 2.3 7>}
  n11 = MaterialFunctionCall(R=@n10.R, G=@n10.R, B=@n10.R, A=@n10.R) {materialFunction=&AppendMany}
  n12 = DotProduct(A=Cosine(None=Multiply(A=@n9, B=@n11.RGBA)), B=Constant4Vector() {constant=<rgba 1 1 1 1>}.RGBA)
  n13 = Power(Base=Saturate(None=Divide(A=@n12) {constB=4}))
  n14 = Add(
   A=LinearInterpolate(B=Step(X=@n13) {constY=0.85, desc="\"Blink\""}, Alpha=ScalarParameter() {defaultValue=1, parameterName=Flicker_BlinkAmount})
   B=@n13
  )
  Add(
   A=LinearInterpolate(
    B=MaterialFunctionCall(WidthBase=@n6, WidthTip=@n6, Expand U or V UV Channel=@n1) {materialFunction=&SplineThicken}.WorldPositionOffset
    Alpha=ScalarParameter() {defaultValue=1, parameterName=Use_SplineThicken}
   ) {constB=0}
   B=Multiply(A=VectorParameter() {parameterName=WPO_Offset, defaultValue=<rgba 0 5 3 1>}.RGB, B=Saturate(None=@n14))
  )
  n15 = Time()
  n16 = Sine(None=@n15) {period=0.66}
  n17 = Sine(None=@n15) {period=0.53}
  n18 = Multiply(A=Multiply(A=ConstantBiasScale(None=@n16), B=ConstantBiasScale(None=@n17)), B=ConstantBiasScale(None=Sine(None=@n15) {period=0.87}))
  n19 = ScalarParameter() {defaultValue=3, parameterName=Divisions}
  Divide(A=Round(None=Multiply(A=@n18, B=@n19) {constB=10}), B=@n19)
  MaterialFunctionCall(Emissive=VectorParameter() {parameterName=Color, defaultValue=<rgba 0.02 0.02 0.02 1>}.RGB) {materialFunction=&MF_FX_EmissiveAdjustments}
  n20 = OneMinus(None=VertexColor().A)
  n21 = StaticBoolParameter() {
   desc="If True, then:\r\nBLACK = 'use Multiplier' and WHITE = 'don't use Multiplier'... \r\n\r\nIt is inverted from a standard mask because vert alpha is 'white' by default. \r\nSo although less intuitive, it means you only have to vert paint the parts you want, instead of having to paint everything. "
   group=Spline Settings
   parameterName=UseVertAlphaThicknessMultiplier?
   sortPriority=60
  }
  n22 = StaticSwitch(
   True=LinearInterpolate(
    B=ScalarParameter() {defaultValue=1, group=Spline Settings, parameterName=ThicknessMult_byBlackVertAlphaMask, sortPriority=61}
    Alpha=Saturate(None=@n20)
   ) {constA=1}
   False=Constant() {r=1}
   Value=@n21
  )
  Multiply(B=@n22)
  Multiply(B=@n22)
  n23 = VectorParameter() {parameterName=SplineTex_ScalePanSpeed, defaultValue=<rgba 0.2 1 1.5 0>}
  n24 = Panner(Coordinate=Multiply(A=TextureCoordinate(), B=AppendVector(A=@n23.R, B=@n23.R)), Speed=AppendVector(A=@n23.B, B=Constant()))
  n25 = TextureObjectParameter() {parameterName=Tex_SplineAlpha, texture=/SparksCosmetics/Emotes/Guitar/ShakeCrunch/Effects/Textures/T_ShakeCrunch_LightningSpline_AlphaMask}
  n26 = ScalarParameter() {defaultValue=0.05, parameterName=EdgeMask_Size}
  n27 = MaterialFunctionCall(Float2=TextureCoordinate()) {materialFunction=&BreakOutFloat2Components}
  n28 = Step(Y=@n26, X=OneMinus(None=@n27.R))
  n29 = Step(Y=@n26, X=@n27.R)
  Saturate(None=Multiply(A=TextureSample(UVs=@n24, Tex=@n25).R, B=Multiply(A=@n28, B=@n29)))
  ScalarParameter() {defaultValue=6, group=Spline Settings, parameterName=Spline Width Tip, sortPriority=1}
 ]
}
