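# Niagara comic-book "BAM" sprite effect for the ShakeCrunch guitar emote in SparksCosmetics.
# BAM spawns one large sprite at a raised point, keeps it small and transparent for over half its life, then pops it visible with a bouncing scale curve.
# Delayed BAM_Shards bursts magenta sprites near the end of that life, with point-origin velocity that decelerates as they shrink.
# Camera offsets layer the two emitters; the declared user variables are not referenced by the modules shown.
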NiagaraSystem /SparksCosmetics/Emotes/Guitar/ShakeCrunch/Effects/NS_ShakeCrunch_ComicBam
dependencies {usedDataInterfaces [&NiagaraDataInterfaceCurve], usedDynamicInputs [], usedModules [&SystemState]}
emitters {
 BAM {
  emitterUpdateScript [
   &EmitterState(
    Inactive Response=::NewEnumerator1
    Life Cycle Mode=::NewEnumerator1
    Loop Behavior=::NewEnumerator1
    Loop Duration=1.05
    Scalability Mode=::NewEnumerator1
    UseLoopDelay=false
   )
   &SpawnBurst_Instantaneous(Use Spawn Probability=false, UseLoopCountLimit=false)
  ]
  particleSpawnScript [
   &InitializeParticle(
    Color Mode=::NewEnumerator0
    Lifetime=1.05
    Position=<0 0 165>
    Position Mode=::NewEnumerator1
    Sprite Rotation Mode=::NewEnumerator2
    Sprite Size Mode=::NewEnumerator3
    Uniform Sprite Size=220
    UsePositionOffset=false
    Write Material Random=true
   )
   &CameraOffset(Camera Offset Amount=-50)
  ]
  particleUpdateScript [
   &ParticleState()
   &ScaleSpriteSize(Uniform Curve Sprite Scale=DataInterface() {propertyValues {Curve {keys [
        {interpMode=RCIM_Cubic, leaveTangentWeight=0.1655, tangentMode=RCTM_User, time=0.0036, value=0.2}
        {arriveTangentWeight=0.2768, interpMode=RCIM_Cubic, leaveTangentWeight=0.0558, tangentMode=RCTM_User, time=0.6, value=0.2}
        {arriveTangent=1.963, arriveTangentWeight=0.1454, interpMode=RCIM_Cubic, leaveTangent=1.963, leaveTangentWeight=0.1645, time=0.7328, value=1}
        <at 0.8292 0.65 RCIM_Cubic>
        {arriveTangent=3.3186, arriveTangentWeight=0.254, interpMode=RCIM_Cubic, leaveTangent=3.3186, tangentMode=RCTM_User, time=1.0008, value=1}
       ]}}})
   &ScaleColor(
    Scale Alpha=&FloatFromCurve(FloatCurve=DataInterface() {propertyValues {Curve {keys [
         {interpMode=RCIM_Cubic, tangentMode=RCTM_User, time=0, value=0}
         {arriveTangentWeight=0.2357, interpMode=RCIM_Cubic, leaveTangentWeight=0.2357, tangentMode=RCTM_User, time=0.63, value=0}
         {arriveTangentWeight=0.1491, interpMode=RCIM_Cubic, leaveTangentWeight=0.2236, tangentMode=RCTM_User, time=0.6653, value=1}
         {interpMode=RCIM_Cubic, tangentMode=RCTM_User, time=1, value=1}
        ]}}})
    Scale RGB=<0.5 0.5 0.5>
    ScaleA=true
    ScaleRGB=true
   )
  ]
  rendererClasses [&NiagaraSpriteRendererProperties]
  renderers [{rendererClass=&NiagaraSpriteRendererProperties}]
 }
 BAM_Shards {
  emitterUpdateScript [
   &EmitterState(Life Cycle Mode=::NewEnumerator1, Loop Behavior=::NewEnumerator1, Loop Delay=0.9656, Loop Duration=0.3203, UseLoopDelay=true)
   &SpawnBurst_Instantaneous(Spawn Count=8, Spawn Time=0.0381, Use Spawn Probability=false, UseLoopCountLimit=false)
  ]
  particleSpawnScript [
   &InitializeParticle(
    Color Maximum=<rgba 0.4354 0.2096 0.4375 1>
    Color Minimum=<rgba 0.2641 0.0014 0.2708 1>
    Color Mode=::NewEnumerator2
    Lifetime=&RandomRangeFloat(Minimum=0.25, Maximum=0.3)
    Position Offset=<0 0 165>
    Sprite Size=&RandomRangeVector2D(Minimum=<75 100>, Maximum=<50 50>)
    Sprite Size Mode=::NewEnumerator4
    UsePositionOffset=true
    Write Material Random=true
   )
   &ShapeLocation(MaskSpawnGroup=false, Sphere Distribution=::NewEnumerator3, Sphere Radius=20, UseOwnerScale=true)
   &AddVelocity(
    Module.Use Falloff Distance=false
    Origin Offset=<0 0 165>
    Override Seed=false
    Velocity Mode=::NewEnumerator1
    Velocity Speed=&RandomRangeFloat(Minimum=-200, Maximum=700)
   )
   &CameraOffset(Camera Offset Amount=-100)
  ]
  particleUpdateScript [
   &ParticleState()
   &ScaleSpriteSize(Uniform Curve Sprite Scale=DataInterface() {propertyValues {Curve {keys [
        {arriveTangent=0.2316, interpMode=RCIM_Cubic, leaveTangent=0.2316, tangentMode=RCTM_User, time=0, value=0}
        {arriveTangent=1.5, interpMode=RCIM_Cubic, leaveTangent=1.5, time=0.2, value=1.2}
        {arriveTangent=-2.4788, interpMode=RCIM_Cubic, leaveTangent=-2.4788, tangentMode=RCTM_User, time=0.5, value=0.75}
        {arriveTangent=-0.4464, interpMode=RCIM_Cubic, leaveTangent=-0.4464, tangentMode=RCTM_User, time=1, value=0}
        <at 1.5 0 RCIM_Cubic>
       ]}}})
   &ScaleColor(
    Scale Alpha=&FloatFromCurve(FloatCurve=DataInterface() {propertyValues {Curve {keys [{interpMode=RCIM_Cubic, tangentMode=RCTM_User, time=0, value=1}, {interpMode=RCIM_Cubic, tangentMode=RCTM_User, time=1, value=0}]}}})
    Scale RGB=&VectorFromFloat(Value=1)
    ScaleA=false
    ScaleRGB=true
   )
   &ScaleVelocity(Velocity Scale=&VectorFromFloat(Value=&FloatFromCurve(
      FloatCurve=DataInterface() {propertyValues {Curve {keys [
          {arriveTangent=-0.0199, interpMode=RCIM_Cubic, leaveTangent=-0.0199, tangentMode=RCTM_User, time=0, value=1}
          {arriveTangent=-1, interpMode=RCIM_Cubic, leaveTangent=-1, time=0.3, value=0.225}
          {arriveTangent=-0.1499, interpMode=RCIM_Cubic, leaveTangent=-0.1499, tangentMode=RCTM_User, time=1, value=0}
         ]}}}
      Scale Curve=2
     )))
   &SolveForcesAndVelocity(Clamp Velocity=false, Limit Acceleration=false)
  ]
  rendererClasses [&NiagaraSpriteRendererProperties]
  renderers [{rendererClass=&NiagaraSpriteRendererProperties}]
 }
}
summary {
 systemName=NS_ShakeCrunch_ComicBam
 emitters [
  {emitterName=BAM, rendererClasses [&NiagaraSpriteRendererProperties]}
  {emitterName=BAM_Shards, rendererClasses [&NiagaraSpriteRendererProperties]}
 ]
 userVariables [
  User.Interpolate Intensity {type=NiagaraFloat, defaultValue=3, description=""}
  User.IsFrontEnd {type=NiagaraBool, defaultValue=true}
  User.IsLargeBody {type=NiagaraBool}
 ]
}
