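# Niagara lightning burst for the ShakeCrunch guitar emote in Sparks cosmetics, mostly yellow-orange with a purple ring.
# An opening center flash mesh, expanding ring sprites and a cone spray of bolt sprites accompany four large lightning meshes with a light renderer.
# A delayed Lightning_Disperse burst follows with four enlarging, fading meshes.
# Both lightning emitters assign mesh variants by particle partition and use stepped curves to jitter Z scale for flicker.
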NiagaraSystem /SparksCosmetics/Emotes/Guitar/ShakeCrunch/Effects/NS_ShakeCrunch_Lightning
dependencies {
 usedDataInterfaces [&NiagaraDataInterfaceArrayInt32, &NiagaraDataInterfaceCurve, &NiagaraDataInterfaceVectorCurve]
 usedDynamicInputs []
 usedModules [&SystemState]
}
emitters {
 Big_Lightning {
  emitterUpdateScript [
   &EmitterState(Inactive Response=::NewEnumerator1, Life Cycle Mode=::NewEnumerator1, Loop Behavior=::NewEnumerator1, Loop Duration=0.55, UseLoopDelay=false)
   &SpawnBurst_Instantaneous(Spawn Count=4, Spawn Time=0.05, Use Spawn Probability=false, UseLoopCountLimit=false)
   &SpawnRate(SpawnRate=3, Use Spawn Probability=false) {enabled=false}
  ]
  particleSpawnScript [
   &InitializeParticle(
    Color=<rgba 3 2 0 1>
    Lifetime Max=1.75
    Lifetime Mode=::NewEnumerator1
    Mass Max=0.5
    Mass Min=0.5
    Mass Mode=::NewEnumerator3
    Mesh Scale Mode=::NewEnumerator3
    Mesh Uniform Scale=0.65
    UsePositionOffset=false
    Write Material Random=true
   )
   &ShapeLocation(MaskSpawnGroup=false, Override Seed=false, Sphere Radius=1, UseOwnerScale=true)
   &PartitionParticles(Number Of Partitions=4, Particles Per Partition=1, Partition Mode:ENiagara_PartitionMode=::NewEnumerator1)
   /SetVariables(Particles.MeshIndex=&SelectIntFromArray(Direct Array Index=$Particles.PartitionParticles.Partition, Int Selection Array=DataInterface() {propertyValues {IntData [0,1,2,3]}})) {isSetParameters=true, module=SetVariables_#1}
  ]
  particleUpdateScript [
   &ParticleState()
   &ScaleMeshSize(Scale Factor=&VectorFromFloat(Value=&FloatFromCurve(FloatCurve=DataInterface() {propertyValues {Curve {keys [
          {arriveTangent=1.4825, interpMode=RCIM_Cubic, leaveTangent=1.4825, tangentMode=RCTM_User, time=0, value=0.45}
          {arriveTangent=4.2063, interpMode=RCIM_Cubic, leaveTangent=4.2063, tangentMode=RCTM_User, time=0.0389, value=0.9553}
          {arriveTangentWeight=0.1021, interpMode=RCIM_Cubic, leaveTangentWeight=0.3979, tangentMode=RCTM_User, time=0.093, value=1.55}
          {arriveTangentWeight=0.0637, interpMode=RCIM_Cubic, leaveTangentWeight=0.3605, tangentMode=RCTM_User, time=0.2352, value=1}
          {arriveTangent=-1.3075, interpMode=RCIM_Cubic, leaveTangent=-1.3075, time=0.8312, value=0.7795}
          {arriveTangent=0.0224, interpMode=RCIM_Cubic, leaveTangent=0.0224, tangentMode=RCTM_User, time=1, value=0}
         ]}}}))) {module=ScaleMeshSize001}
   &ScaleMeshSize(Scale Factor=&Lerp_Vector(
     A=<1 1 1>
     B=<1 1 1.5>
     Alpha=&FloatFromCurve(FloatCurve=DataInterface() {propertyValues {Curve {keys [<at 0 1 RCIM_Constant>,<at 0.0242 0.05 RCIM_Constant>,<at 0.0686 0.75 RCIM_Constant>,<at 0.12 0.4 RCIM_Constant>,<at 0.1729 0.9 RCIM_Constant>,<at 0.232 0.15 RCIM_Constant>,<at 0.3064 0.9 RCIM_Constant>,<at 0.3634 0.1 RCIM_Constant>,<at 0.4235 0.55 RCIM_Constant>,<at 0.5274 0 RCIM_Constant>,<at 0.5559 0.9 RCIM_Constant>,<at 0.6323 0.05 RCIM_Constant>,<at 0.6822 0.9 RCIM_Constant>,<at 0.7983 0 RCIM_Constant>,<at 0.9164 0.9 RCIM_Constant>,<at 1 0 RCIM_Constant>]}}})
    ))
   &Drag(Drag=0.25, Use Linear Drag=true, Use Rotational Drag=false)
   &ScaleColor(
    Scale Alpha=&FloatFromCurve(FloatCurve=DataInterface() {propertyValues {Curve {keys [<at -0.0046 1.0113 RCIM_Cubic>,<at 1.0011 0.002 RCIM_Cubic>]}}})
    ScaleA=true
    ScaleRGB=true
   )
   &SolveForcesAndVelocity(Clamp Velocity=false, Limit Acceleration=false)
  ]
  rendererClasses [&NiagaraMeshRendererProperties, &NiagaraLightRendererProperties]
  renderers [
   {rendererClass=&NiagaraMeshRendererProperties, rendererIndex=0}
   {rendererClass=&NiagaraMeshRendererProperties, rendererIndex=1}
   {rendererClass=&NiagaraLightRendererProperties, rendererIndex=2}
  ]
 }
 CenterFlash {
  emitterUpdateScript [
   &EmitterState(Life Cycle Mode=::NewEnumerator1, Loop Behavior=::NewEnumerator1, Loop Duration=2, UseLoopDelay=false)
   &SpawnBurst_Instantaneous(Use Spawn Probability=false, UseLoopCountLimit=false)
  ]
  particleSpawnScript [&InitializeParticle(
    Color=<rgba 2 1 0 1>
    Lifetime=0.25
    Mesh Scale Mode=::NewEnumerator3
    Mesh Uniform Scale=0.75
    Sprite Size Mode=::NewEnumerator3
    Uniform Sprite Size=100
    UsePositionOffset=true
    Write Material Random=true
   )]
  particleUpdateScript [
   &ParticleState()
   &ScaleMeshSize(Scale Factor=&Lerp_Vector(
     A=<1 1 1>
     B=<1 1 1>
     Alpha=&FloatFromCurve(FloatCurve=DataInterface() {propertyValues {Curve {keys [<at 0 1 RCIM_Constant>,<at 0.0242 0.05 RCIM_Constant>,<at 0.2 0.85 RCIM_Constant>,<at 0.3634 0.1 RCIM_Constant>,<at 0.6 0.85 RCIM_Constant>,<at 0.7 0.2 RCIM_Constant>,<at 0.9164 0.9 RCIM_Constant>,<at 1 0 RCIM_Constant>]}}})
    )) {module=ScaleMeshSize001}
   &ScaleMeshSize(Scale Factor=&VectorFromCurve(VectorCurve=DataInterface() {propertyValues {
       XCurve {keys [
         {tangentMode=RCTM_User, time=0, value=1.725}
         {arriveTangent=-2.6013, interpMode=RCIM_Cubic, leaveTangent=-2.6013, tangentMode=RCTM_User, time=0.4, value=0.5}
         {arriveTangent=-0.2543, interpMode=RCIM_Cubic, leaveTangent=-0.2543, tangentMode=RCTM_User, time=1, value=0}
        ]}
       YCurve {keys [
         {tangentMode=RCTM_User, time=0, value=0.75}
         {arriveTangent=-0.0405, interpMode=RCIM_Cubic, leaveTangent=-0.0405, tangentMode=RCTM_User, time=0.6, value=0.7}
         {arriveTangent=-0.1088, interpMode=RCIM_Cubic, leaveTangent=-0.1088, tangentMode=RCTM_User, time=1, value=0}
        ]}
       ZCurve {keys [
         {tangentMode=RCTM_User, time=0, value=0.75}
         {arriveTangent=-0.0405, interpMode=RCIM_Cubic, leaveTangent=-0.0405, tangentMode=RCTM_User, time=0.6, value=0.725}
         {arriveTangent=-0.1088, interpMode=RCIM_Cubic, leaveTangent=-0.1088, tangentMode=RCTM_User, time=1, value=0}
        ]}
      }}))
   &MeshRotationForce(Rotation=<45 0 0>)
   &SolveForcesAndVelocity(Clamp Velocity=false, Limit Acceleration=false)
  ]
  rendererClasses [&NiagaraMeshRendererProperties]
  renderers [
   {rendererClass=&NiagaraMeshRendererProperties, rendererIndex=0}
   {rendererClass=&NiagaraMeshRendererProperties, rendererIndex=1}
  ]
 }
 Lightning_Disperse {
  emitterUpdateScript [
   &EmitterState(
    Inactive Response=::NewEnumerator1
    Life Cycle Mode=::NewEnumerator1
    Loop Behavior=::NewEnumerator1
    Loop Delay=0.55
    Loop Duration=0.2
    UseLoopDelay=true
   )
   &SpawnBurst_Instantaneous(Spawn Count=4, Use Spawn Probability=false, UseLoopCountLimit=false)
   &SpawnRate(SpawnRate=3, Use Spawn Probability=false) {enabled=false}
  ]
  particleSpawnScript [
   &InitializeParticle(
    Color=<rgba 0.6927 0.4331 0 1>
    Lifetime Max=1.75
    Lifetime Mode=::NewEnumerator1
    Mass Max=0.5
    Mass Min=0.5
    Mass Mode=::NewEnumerator3
    Mesh Scale Mode=::NewEnumerator3
    Mesh Uniform Scale=0.65
    UsePositionOffset=false
    Write Material Random=true
   )
   &ShapeLocation(MaskSpawnGroup=false, Override Seed=false, Sphere Radius=1, UseOwnerScale=true)
   &PartitionParticles(Number Of Partitions=4, Particles Per Partition=1, Partition Mode:ENiagara_PartitionMode=::NewEnumerator1)
   /SetVariables(Particles.MeshIndex=&SelectIntFromArray(Direct Array Index=$Particles.PartitionParticles.Partition, Int Selection Array=DataInterface() {propertyValues {IntData [0,1,2,3]}})) {isSetParameters=true, module=SetVariables_#1}
  ]
  particleUpdateScript [
   &ParticleState()
   &ScaleMeshSize(Scale Factor=&VectorFromFloat(Value=&FloatFromCurve(FloatCurve=DataInterface() {propertyValues {Curve {keys [
          {arriveTangent=13.8106, interpMode=RCIM_Cubic, leaveTangent=13.8106, tangentMode=RCTM_User, time=0, value=0.75}
          {arriveTangentWeight=0.1374, interpMode=RCIM_Cubic, leaveTangentWeight=1.2369, tangentMode=RCTM_User, time=0.1, value=1.75}
          {interpMode=RCIM_Cubic, tangentMode=RCTM_User, time=1, value=1.75}
         ]}}}))) {module=ScaleMeshSize001}
   &ScaleMeshSize(Scale Factor=&Lerp_Vector(
     A=<1 1 1>
     B=<1 1 1.5>
     Alpha=&FloatFromCurve(FloatCurve=DataInterface() {propertyValues {Curve {keys [<at 0 1 RCIM_Constant>,<at 0.0242 0.8578 RCIM_Constant>,<at 0.0686 0.9225 RCIM_Constant>,<at 0.12 0.9176 RCIM_Constant>,<at 0.1729 0.9 RCIM_Constant>,<at 0.232 0.8402 RCIM_Constant>,<at 0.3064 0.9 RCIM_Constant>,<at 0.3634 0.7902 RCIM_Constant>,<at 0.4235 0.8245 RCIM_Constant>,<at 0.5274 0.7765 RCIM_Constant>,<at 0.5559 0.9 RCIM_Constant>,<at 0.6323 0.8265 RCIM_Constant>,<at 0.6822 0.9 RCIM_Constant>,<at 0.7983 0.7765 RCIM_Constant>,<at 0.9164 0.9 RCIM_Constant>,<at 1 0.7765 RCIM_Constant>]}}})
    ))
   &ScaleColor(
    Scale Alpha=&FloatFromCurve(FloatCurve=DataInterface() {propertyValues {Curve {keys [<at -0.0046 1.0113 RCIM_Cubic>,<at 1.0011 0.002 RCIM_Cubic>]}}})
    ScaleA=true
    ScaleRGB=true
   )
   &ScaleVelocity(Velocity Scale=&VectorFromFloat(Value=&FloatFromCurve(FloatCurve=DataInterface() {propertyValues {Curve {keys [
          {tangentMode=RCTM_User, time=0, value=1}
          {arriveTangent=-0.0405, interpMode=RCIM_Cubic, leaveTangent=-0.0405, tangentMode=RCTM_User, time=0.7009, value=1}
          {arriveTangent=-8.0089, interpMode=RCIM_Cubic, leaveTangent=-8.0089, tangentMode=RCTM_User, time=1, value=0}
         ]}}})))
   &SolveForcesAndVelocity(Clamp Velocity=false, Limit Acceleration=false)
  ]
  rendererClasses [&NiagaraMeshRendererProperties]
  renderers [{rendererClass=&NiagaraMeshRendererProperties, rendererIndex=0}]
 }
 Ring {
  emitterUpdateScript [
   &EmitterState(Life Cycle Mode=::NewEnumerator1, Loop Behavior=::NewEnumerator1, Loop Duration=1, UseLoopDelay=false)
   &SpawnBurst_Instantaneous(Use Spawn Probability=false, UseLoopCountLimit=false)
   &SpawnBurst_Instantaneous(Spawn Time=0.1, Use Spawn Probability=false, UseLoopCountLimit=false) {module=SpawnBurst_Instantaneous001}
  ]
  particleSpawnScript [&InitializeParticle(
    Color=<rgba 0.2706 0.0809 0.9288 1>
    Lifetime=0.45
    Sprite Size Mode=::NewEnumerator3
    Uniform Sprite Size=120
    UsePositionOffset=false
    Write Material Random=true
   )]
  particleUpdateScript [
   &ParticleState()
   &ScaleSpriteSize(Uniform Curve Sprite Scale=DataInterface() {propertyValues {Curve {keys [
        {arriveTangent=4.7506, interpMode=RCIM_Cubic, leaveTangent=4.7506, tangentMode=RCTM_User, time=0, value=0.2}
        {arriveTangent=0.6731, interpMode=RCIM_Cubic, leaveTangent=0.6731, tangentMode=RCTM_User, time=0.3, value=0.85}
        <at 1 1 RCIM_Cubic>
       ]}}})
   &DynamicMaterialParameters(
    Index 0 Param 1=&FloatFromCurve(FloatCurve=DataInterface() {propertyValues {Curve {keys [
         {interpMode=RCIM_Cubic, leaveTangentWeight=0.2315, tangentMode=RCTM_User, time=0, value=0}
         {arriveTangent=0.7978, interpMode=RCIM_Cubic, leaveTangent=0.7978, tangentMode=RCTM_User, time=0.3, value=0.1}
         {
          arriveTangent=1.2959
          arriveTangentWeight=0.0543
          interpMode=RCIM_Cubic
          leaveTangent=1.2959
          leaveTangentWeight=0.3258
          tangentMode=RCTM_User
          time=0.6
          value=0.75
         }
         {arriveTangent=0.7322, interpMode=RCIM_Cubic, leaveTangent=0.7322, tangentMode=RCTM_User, time=1, value=1}
        ]}}})
    Module.Param0WriteEnabled=true
    Module.Param1WriteEnabled=false
    Module.Param2WriteEnabled=false
    Module.Param3WriteEnabled=false
   )
  ]
  rendererClasses [&NiagaraSpriteRendererProperties]
  renderers [{rendererClass=&NiagaraSpriteRendererProperties}]
 }
 Start_BoltBurst {
  emitterUpdateScript [
   &EmitterState(Life Cycle Mode=::NewEnumerator1, Loop Behavior=::NewEnumerator1, Loop Duration=0.55, UseLoopDelay=false)
   &SpawnRate(SpawnRate=1000, Use Spawn Probability=false) {enabled=false}
   &SpawnBurst_Instantaneous(Spawn Count=9, Spawn Time=0.05, Use Spawn Probability=false, UseLoopCountLimit=false)
  ]
  particleSpawnScript [
   &InitializeParticle(
    Color=<rgba 2 1.2 0 1>
    Lifetime=&RandomRangeFloat(Minimum=0.5, Maximum=0.8)
    Sprite Size=<15 60>
    Sprite Size Mode=::NewEnumerator4
    UsePositionOffset=false
    Write Material Random=true
   )
   &AddVelocity(
    Cone Angle=120
    Cone Axis=<0 -1 0.5>
    Inner Cone Angle=30
    Module.UseConeFalloff=false
    Module.UseVelDistribution=false
    Override Seed=false
    Rotation Angle=-101.5166
    Rotation Axis:Vector3f=<0 0 0>
    Rotation Mode=::NewEnumerator1
    Velocity Mode=::NewEnumerator2
    Velocity Speed=&RandomRangeFloat(Minimum=1000, Maximum=1600)
   )
  ]
  particleUpdateScript [
   &ParticleState()
   &ScaleSpriteSize(Uniform Curve Sprite Scale=DataInterface() {propertyValues {Curve {keys [
        {interpMode=RCIM_Cubic, tangentMode=RCTM_User, time=0, value=1}
        {arriveTangent=-1.5, interpMode=RCIM_Cubic, leaveTangent=-1.5, time=0.3, value=0.75}
        {arriveTangent=-0.957, interpMode=RCIM_Cubic, leaveTangent=-0.957, tangentMode=RCTM_User, time=0.5, value=0.25}
        {interpMode=RCIM_Cubic, tangentMode=RCTM_User, time=1, value=0}
       ]}}})
   &ScaleColor(
    Scale Alpha=&FloatFromCurve(FloatCurve=DataInterface() {propertyValues {Curve {keys [<at -0.0046 1.0113 RCIM_Cubic>,<at 1.0011 0.002 RCIM_Cubic>]}}}, Scale Curve=0.5)
    ScaleA=true
    ScaleRGB=false
   ) {enabled=false}
   &Drag(Drag=9, Use Linear Drag=true, Use Rotational Drag=true)
   &ScaleVelocity(Velocity Scale=&VectorFromFloat(Value=&FloatFromCurve(
      FloatCurve=DataInterface() {propertyValues {Curve {keys [
          {interpMode=RCIM_Cubic, tangentMode=RCTM_User, time=0, value=1}
          {
           arriveTangent=-0.255
           arriveTangentWeight=0.0471
           interpMode=RCIM_Cubic
           leaveTangent=-0.255
           leaveTangentWeight=0.4243
           tangentMode=RCTM_User
           time=0.2
           value=0.15
          }
          {interpMode=RCIM_Cubic, tangentMode=RCTM_User, time=1, value=0}
         ]}}}
      Scale Curve=1.5
     )))
   &SolveForcesAndVelocity(Clamp Velocity=false, Limit Acceleration=false)
   &ScaleColor(ScaleA=true, ScaleRGB=true) {module=ScaleColor001}
  ]
  rendererClasses [&NiagaraSpriteRendererProperties]
  renderers [{rendererClass=&NiagaraSpriteRendererProperties}]
 }
}
summary {
 systemName=NS_ShakeCrunch_Lightning
 emitters [
  {emitterName=Lightning_Disperse, rendererClasses [&NiagaraMeshRendererProperties]}
  {emitterName=Start_BoltBurst, rendererClasses [&NiagaraSpriteRendererProperties]}
  {emitterName=Big_Lightning, rendererClasses [&NiagaraMeshRendererProperties, &NiagaraLightRendererProperties]}
  {emitterName=CenterFlash, rendererClasses [&NiagaraMeshRendererProperties]}
  {emitterName=Ring, rendererClasses [&NiagaraSpriteRendererProperties]}
 ]
 userVariables []
}
